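# Translucent unlit spark material for Niagara sprites in the Meticulous emote effect, with responsive AA enabled.
# A parameter blends between a tint color and the spark texture's RGB, multiplied by particle color and passed through two chained MF_FX_EmissiveAdjustments calls for the emissive.
# The texture's red channel, softened by depth fade and scaled by particle alpha, forms the opacity.
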Material /MistRaven/Athena/Effects/Particles/Emotes/NS_Emote_Meticulous/M_Emote_Meticulous_Spark
graph {
 properties {
  bEnableResponsiveAA=true
  bUsedWithNiagaraSprites=true
  blendMode=BLEND_Translucent
  pixelDepthOffsetMode=PDOM_Legacy
  shadingModel=MSM_Unlit
  usedShadingModels=Unlit
 }
 nodes [
  n1 = TextureSampleParameter2D() {parameterName=Tex, texture=/Game/Weapons/FORT_Rifles/Effects/Textures/spark}
  n2 = LinearInterpolate(A=VectorParameter() {parameterName=Color, defaultValue=<rgba 1 1 1 1>}.RGB, B=@n1.RGB, Alpha=ScalarParameter() {parameterName=Use Tex Color})
  n3 = ParticleColor()
  MaterialFunctionCall(Emissive=MaterialFunctionCall(Emissive=Multiply(A=@n2, B=@n3.RGB)) {materialFunction=&MF_FX_EmissiveAdjustments}.Result) {materialFunction=&MF_FX_EmissiveAdjustments}
  Multiply(A=@n3.A, B=DepthFade(Opacity=@n1.R, FadeDistance=ScalarParameter() {parameterName=FadeDistance}))
 ]
}
